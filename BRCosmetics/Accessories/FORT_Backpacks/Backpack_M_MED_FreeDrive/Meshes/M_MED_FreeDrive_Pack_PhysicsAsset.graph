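# Physics asset for the MED FreeDrive backpack cosmetic mesh, setting up bodies for its dangling or jiggling bones.
# A root sphere has no constraint, while twelve small dyn_fri capsules attach to the box_fri capsule with swing and twist locked.
# Five of them (02, 03, 04, 06, 11) pair with offset _rot capsules through unlocked angular constraints carrying small limit angles.
# The dyn_fri_11 pair lists its bones in reverse order compared with the other four.
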PhysicsAsset /BRCosmetics/Accessories/FORT_Backpacks/Backpack_M_MED_FreeDrive/Meshes/M_MED_FreeDrive_Pack_PhysicsAsset
bodies {
 box_fri [{shapeType=Capsule, length=1, radius=1, shapeName=box_fri_capsule}]
 dyn_fri_01 [{shapeType=Capsule, length=0.2, radius=0.5, shapeName=dyn_fri_01_capsule}]
 dyn_fri_02 [{shapeType=Capsule, length=0.2, radius=0.5, shapeName=dyn_fri_02_capsule}]
 dyn_fri_02_rot [{shapeType=Capsule, length=0.5, radius=0.5, shapeName=dyn_fri_02_rot_capsule, center=<2.9363 -0.2987 -0.6232>}]
 dyn_fri_03 [{shapeType=Capsule, length=0.2, radius=0.5, shapeName=dyn_fri_03_capsule}]
 dyn_fri_03_rot [{shapeType=Capsule, length=0.5, radius=0.5, shapeName=dyn_fri_03_rot_capsule, center=<3.1375 0 -0.4813>}]
 dyn_fri_04 [{shapeType=Capsule, length=0.2, radius=0.5, shapeName=dyn_fri_04_capsule}]
 dyn_fri_04_rot [{shapeType=Capsule, length=0.5, radius=0.5, shapeName=dyn_fri_04_rot_capsule, center=<4.0553 -1.5118 0>}]
 dyn_fri_05 [{shapeType=Capsule, length=0.2, radius=0.5, shapeName=dyn_fri_05_capsule}]
 dyn_fri_06 [{shapeType=Capsule, length=0.2, radius=0.5, shapeName=dyn_fri_06_capsule}]
 dyn_fri_06_rot [{shapeType=Capsule, length=0.5, radius=0.5, shapeName=dyn_fri_06_rot_capsule, center=<3.9572 0 0.5275>}]
 dyn_fri_07 [{shapeType=Capsule, length=0.2, radius=0.5, shapeName=dyn_fri_07_capsule}]
 dyn_fri_08 [{shapeType=Capsule, length=0.2, radius=0.5, shapeName=dyn_fri_08_capsule}]
 dyn_fri_09 [{shapeType=Capsule, length=0.2, radius=0.5, shapeName=dyn_fri_09_capsule}]
 dyn_fri_10 [{shapeType=Capsule, length=0.2, radius=0.5, shapeName=dyn_fri_10_capsule}]
 dyn_fri_11 [{shapeType=Capsule, length=0.2, radius=0.5, shapeName=dyn_fri_11_capsule}]
 dyn_fri_11_rot [{shapeType=Capsule, length=0.5, radius=0.5, shapeName=dyn_fri_11_rot_capsule, center=<2.0476 0.0277 0.1606>}]
 dyn_fri_12 [{shapeType=Capsule, length=0.2, radius=0.5, shapeName=dyn_fri_12_capsule}]
 root [{shapeType=Sphere, radius=21.5, shapeName=root_sphere, center=<0 -11.428 0>}]
}
constraints [
 {
  bone1Name=box_fri
  bone2Name=dyn_fri_01
  swing1LimitDegrees=15
  swing1Motion=Locked
  swing2LimitDegrees=15
  swing2Motion=Locked
  twistLimitDegrees=4
  twistMotion=Locked
 }
 {
  bone1Name=box_fri
  bone2Name=dyn_fri_02
  swing1LimitDegrees=15
  swing1Motion=Locked
  swing2LimitDegrees=15
  swing2Motion=Locked
  twistLimitDegrees=4
  twistMotion=Locked
 }
 {
  bone1Name=box_fri
  bone2Name=dyn_fri_03
  swing1LimitDegrees=15
  swing1Motion=Locked
  swing2LimitDegrees=15
  swing2Motion=Locked
  twistLimitDegrees=4
  twistMotion=Locked
 }
 {
  bone1Name=box_fri
  bone2Name=dyn_fri_04
  swing1LimitDegrees=15
  swing1Motion=Locked
  swing2LimitDegrees=15
  swing2Motion=Locked
  twistLimitDegrees=4
  twistMotion=Locked
 }
 {
  bone1Name=box_fri
  bone2Name=dyn_fri_05
  swing1LimitDegrees=15
  swing1Motion=Locked
  swing2LimitDegrees=15
  swing2Motion=Locked
  twistLimitDegrees=4
  twistMotion=Locked
 }
 {
  bone1Name=box_fri
  bone2Name=dyn_fri_06
  swing1LimitDegrees=15
  swing1Motion=Locked
  swing2LimitDegrees=15
  swing2Motion=Locked
  twistLimitDegrees=4
  twistMotion=Locked
 }
 {
  bone1Name=box_fri
  bone2Name=dyn_fri_07
  swing1LimitDegrees=15
  swing1Motion=Locked
  swing2LimitDegrees=15
  swing2Motion=Locked
  twistLimitDegrees=4
  twistMotion=Locked
 }
 {
  bone1Name=box_fri
  bone2Name=dyn_fri_08
  swing1LimitDegrees=15
  swing1Motion=Locked
  swing2LimitDegrees=15
  swing2Motion=Locked
  twistLimitDegrees=4
  twistMotion=Locked
 }
 {
  bone1Name=box_fri
  bone2Name=dyn_fri_09
  swing1LimitDegrees=15
  swing1Motion=Locked
  swing2LimitDegrees=15
  swing2Motion=Locked
  twistLimitDegrees=4
  twistMotion=Locked
 }
 {
  bone1Name=box_fri
  bone2Name=dyn_fri_10
  swing1LimitDegrees=15
  swing1Motion=Locked
  swing2LimitDegrees=15
  swing2Motion=Locked
  twistLimitDegrees=4
  twistMotion=Locked
 }
 {
  bone1Name=box_fri
  bone2Name=dyn_fri_11
  swing1LimitDegrees=15
  swing1Motion=Locked
  swing2LimitDegrees=15
  swing2Motion=Locked
  twistLimitDegrees=4
  twistMotion=Locked
 }
 {
  bone1Name=box_fri
  bone2Name=dyn_fri_12
  swing1LimitDegrees=15
  swing1Motion=Locked
  swing2LimitDegrees=15
  swing2Motion=Locked
  twistLimitDegrees=4
  twistMotion=Locked
 }
 {bone1Name=dyn_fri_02, bone2Name=dyn_fri_02_rot, swing1LimitDegrees=1, swing2LimitDegrees=5, twistLimitDegrees=3}
 {bone1Name=dyn_fri_03, bone2Name=dyn_fri_03_rot, swing1LimitDegrees=6, swing2LimitDegrees=6, twistLimitDegrees=3}
 {bone1Name=dyn_fri_04, bone2Name=dyn_fri_04_rot, swing1LimitDegrees=5, swing2LimitDegrees=1, twistLimitDegrees=3}
 {bone1Name=dyn_fri_11_rot, bone2Name=dyn_fri_11, swing1LimitDegrees=1, swing2LimitDegrees=5, twistLimitDegrees=3}
 {bone1Name=dyn_fri_06, bone2Name=dyn_fri_06_rot, swing1LimitDegrees=1, swing2LimitDegrees=5, twistLimitDegrees=3}
]
